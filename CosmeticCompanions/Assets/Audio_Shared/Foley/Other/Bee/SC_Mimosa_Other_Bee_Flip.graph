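# Sound cue for the bee flip foley of a Mimosa cosmetic companion, classed as a pet creature sound.
# A quality-level root shares two random pools across its four slots: the first two draw from five equally weighted flip waves, the last two from only the first four.
# It uses companion foley attenuation and borrows the quadruped footstep concurrency set.
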SoundCue /CosmeticCompanions/Assets/Audio_Shared/Foley/Other/Bee/SC_Mimosa_Other_Bee_Flip
root=QualityLevel_0
nodes [
 {id=QualityLevel_0, children [Random_2,Random_2,Random_3,Random_3]}
 {id=Random_2, children [WavePlayer_0, WavePlayer_1, WavePlayer_10, WavePlayer_11, WavePlayer_12], properties {weights [1,1,1,1,1]}}
 {id=Random_3, children [WavePlayer_13, WavePlayer_14, WavePlayer_15, WavePlayer_16], properties {weights [1,1,1,1]}}
 {id=WavePlayer_0, properties {soundWaveAssetPtr=/CosmeticCompanions/Assets/Audio_Shared/Foley/Other/Bee/Wavs/Mimosa_Bee_Flip_01}}
 {id=WavePlayer_1, properties {soundWaveAssetPtr=/CosmeticCompanions/Assets/Audio_Shared/Foley/Other/Bee/Wavs/Mimosa_Bee_Flip_02}}
 {id=WavePlayer_10, properties {soundWaveAssetPtr=/CosmeticCompanions/Assets/Audio_Shared/Foley/Other/Bee/Wavs/Mimosa_Bee_Flip_03}}
 {id=WavePlayer_11, properties {soundWaveAssetPtr=/CosmeticCompanions/Assets/Audio_Shared/Foley/Other/Bee/Wavs/Mimosa_Bee_Flip_04}}
 {id=WavePlayer_12, properties {soundWaveAssetPtr=/CosmeticCompanions/Assets/Audio_Shared/Foley/Other/Bee/Wavs/Mimosa_Bee_Flip_05}}
 {id=WavePlayer_13, properties {soundWaveAssetPtr=&Mimosa_Bee_Flip_01}}
 {id=WavePlayer_14, properties {soundWaveAssetPtr=&Mimosa_Bee_Flip_02}}
 {id=WavePlayer_15, properties {soundWaveAssetPtr=&Mimosa_Bee_Flip_03}}
 {id=WavePlayer_16, properties {soundWaveAssetPtr=&Mimosa_Bee_Flip_04}}
]
properties {
 attenuationSettings=/Game/Sounds/Attenuation/AS_Companion_Foley
 maxDistance=1200
 priority=0
 soundClassObject=&Fort_Creatures_Pets
 volumeMultiplier=1
 concurrencySet [/CosmeticCompanions/Assets/Audio_Shared/Foley/Quadraped/CONC_Mimosa_Quadraped_Footstep]
}
